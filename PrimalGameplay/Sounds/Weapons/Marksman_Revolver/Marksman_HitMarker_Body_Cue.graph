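# Sound cue for the Marksman Revolver's body hit marker, the notification for a non-critical hit.
# A random node picks one of three equally weighted hit waves, which passes through a modulator and a short fixed delay.
# It routes to a 2D hit-notification sound class for the instigator and references a shared concurrency set that keeps at most two sounds, stopping the oldest.
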SoundCue /PrimalGameplay/Sounds/Weapons/Marksman_Revolver/Marksman_HitMarker_Body_Cue
root=Delay_0
nodes [
 {id=Delay_0, children [Modulator_0], properties {delayMax=0.1, delayMin=0.1}}
 {id=Modulator_0, children [Random_1]}
 {id=Random_1, children [WavePlayer_3,WavePlayer_4,WavePlayer_5], properties {weights [1,1,1]}}
 {id=WavePlayer_3, properties {soundWaveAssetPtr=&/Marksman_Revolver_Non_Critical_Hit_01}}
 {id=WavePlayer_4, properties {soundWaveAssetPtr=&/Marksman_Revolver_Non_Critical_Hit_02}}
 {id=WavePlayer_5, properties {soundWaveAssetPtr=&/Marksman_Revolver_Non_Critical_Hit_03}}
]
properties {
 maxDistance=2097152
 soundClassObject=&Fort_2DHitNotifications_Instigate_PSM
 volumeMultiplier=0.35
 concurrencyOverrides {maxCount=2, resolutionRule=StopOldest}
 concurrencySet [&HitNotifications_TwoMaximum_StopOldest_VolumeScale]
}
