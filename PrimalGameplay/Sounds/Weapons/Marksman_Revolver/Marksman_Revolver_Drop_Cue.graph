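# Sound cue for the drop sound of the Marksman Revolver weapon.
# A single wave player plays a shared pistol drop loot sample at reduced volume, with no randomization or mixing.
# It uses the weapon drop attenuation, the loot collect sound class and a stop-oldest weapon drop concurrency set.
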SoundCue /PrimalGameplay/Sounds/Weapons/Marksman_Revolver/Marksman_Revolver_Drop_Cue
root=WavePlayer_8
nodes [{id=WavePlayer_8, properties {soundWaveAssetPtr=/Game/Sounds/Fort_GamePlay_Sounds/Loot/pistol_drop_06}}]
properties {
 attenuationSettings=&Fort_Weapon_Drop_Att
 maxDistance=3000
 soundClassObject=&Fort_Loot_Collect
 volumeMultiplier=0.5
 concurrencyOverrides {maxCount=1, resolutionRule=PreventNew}
 concurrencySet [&Weapon_Drop_StopOldest]
}
