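# Sound cue for the equip sound of the Marksman Revolver weapon.
# A single wave player feeds a modulator that applies slight random pitch variation on each play.
# It routes to the player foley sound class with weapon reload attenuation.
# A concurrency override allows one instance at a time per owner.
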SoundCue /PrimalGameplay/Sounds/Weapons/Marksman_Revolver/Marksman_Revolver_Equip_Cue
root=Modulator_0
nodes [
 {id=Modulator_0, children [WavePlayer_1], properties {pitchMax=1.02, pitchMin=0.99}}
 {id=WavePlayer_1, properties {soundWaveAssetPtr=&/Marksman_Revolver_Equip}}
]
properties {
 attenuationSettings=&Fort_Weapon_Reload_Att
 bOverrideConcurrency=true
 maxDistance=2250
 soundClassObject=&Fort_Player_Foley
 volumeMultiplier=1
 attenuationOverrides {falloffDistance=2000, nonSpatializedRadiusStart=200, attenuationShapeExtents {x=500}}
 concurrencyOverrides {bLimitToOwner=true, maxCount=1}
}
