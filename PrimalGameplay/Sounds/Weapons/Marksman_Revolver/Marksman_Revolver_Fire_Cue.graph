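# Marksman Revolver gunfire sound cue, layering close, medium and distant shots under local and global gunfire concurrency.
# A quality switch sends its first level to the full variation set and the other three to a shared, smaller set.
# Each distance layer has its own attenuation and volume/pitch modulation, with the distant layer pitched down.
# Close layers use the 3P fire sound class; medium and far layers use the distant class.
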SoundCue /PrimalGameplay/Sounds/Weapons/Marksman_Revolver/Marksman_Revolver_Fire_Cue
root=QualityLevel_0
nodes [
 {id=Attenuation_0, children [Modulator_0], properties {attenuationSettings=/Game/Sounds/Attenuation/Fort_Weapon_Projectile_Distant_StW_Att}}
 {id=Attenuation_1, children [Modulator_1], properties {attenuationSettings=&Fort_Weapon_Projectile_Close_Athena_Att}}
 {
  id=Attenuation_119
  children [Modulator_2]
  properties {attenuationSettings=&Fort_Weapon_Projectile_Medium_Mono_Athena_Att}
 }
 {id=Attenuation_2, children [Modulator_5], properties {attenuationSettings=&Fort_Weapon_Projectile_Distant_StW_Att}}
 {id=Attenuation_3, children [Modulator_6], properties {attenuationSettings=&Fort_Weapon_Projectile_Close_Athena_Att}}
 {
  id=Attenuation_6
  children [Modulator_3]
  properties {attenuationSettings=&Fort_Weapon_Projectile_Medium_Mono_Athena_Att}
 }
 {id=Mixer_0, children [SoundClass_76, SoundClass_2], properties {inputVolume [0.8,0.3]}}
 {id=Mixer_1, children [SoundClass_1,Mixer_6], properties {inputVolume [1,0.8]}}
 {id=Mixer_2, children [SoundClass_3,Mixer_0], properties {inputVolume [1,0.8]}}
 {id=Mixer_6, children [SoundClass_6,SoundClass_0], properties {inputVolume [0.8,0.8]}}
 {id=Modulator_0, children [WavePlayer_35], properties {pitchMax=0.9, pitchMin=0.8, volumeMax=1.4, volumeMin=1.2}}
 {id=Modulator_1, children [Random_8], properties {pitchMin=0.98, volumeMax=1.4, volumeMin=1.3}}
 {id=Modulator_2, children [Random_5], properties {volumeMax=1.6, volumeMin=1.4}}
 {id=Modulator_3, children [Random_9], properties {volumeMax=1.6, volumeMin=1.4}}
 {id=Modulator_5, children [Random_7], properties {pitchMax=0.9, pitchMin=0.8, volumeMax=1.4, volumeMin=1.2}}
 {id=Modulator_6, children [Random_4], properties {pitchMin=0.98, volumeMax=1.4, volumeMin=1.3}}
 {id=QualityLevel_0, children [Mixer_2,Mixer_1,Mixer_1,Mixer_1]}
 {id=Random_4, children [WavePlayer_15, WavePlayer_16, WavePlayer_17, WavePlayer_18], properties {weights [1,1,1,1]}}
 {id=Random_5, children [WavePlayer_19, WavePlayer_20, WavePlayer_21], properties {weights [1,1,1]}}
 {id=Random_7, children [WavePlayer_22, WavePlayer_23, WavePlayer_24], properties {weights [1,1,1]}}
 {id=Random_8, children [WavePlayer_25, WavePlayer_26, WavePlayer_29], properties {weights [1,1,1]}}
 {id=Random_9, children [WavePlayer_32, WavePlayer_33], properties {weights [1,1]}}
 {
  id=SoundClass_0
  children [Attenuation_0]
  properties {soundClassOverride=&Fort_WeaponRanged_Fire_Distant}
 }
 {
  id=SoundClass_1
  children [Attenuation_1]
  properties {soundClassOverride=&Fort_WeaponRanged_Fire_3P_PSM}
 }
 {
  id=SoundClass_2
  children [Attenuation_2]
  properties {soundClassOverride=&Fort_WeaponRanged_Fire_Distant}
 }
 {
  id=SoundClass_3
  children [Attenuation_3]
  properties {soundClassOverride=&Fort_WeaponRanged_Fire_3P_PSM}
 }
 {
  id=SoundClass_6
  children [Attenuation_6]
  properties {soundClassOverride=&Fort_WeaponRanged_Fire_Distant}
 }
 {
  id=SoundClass_76
  children [Attenuation_119]
  properties {soundClassOverride=&Fort_WeaponRanged_Fire_Distant}
 }
 {id=WavePlayer_15, properties {soundWaveAssetPtr=&/Marksman_Revolver_Fire_Close_01}}
 {id=WavePlayer_16, properties {soundWaveAssetPtr=&/Marksman_Revolver_Fire_Close_02}}
 {id=WavePlayer_17, properties {soundWaveAssetPtr=&/Marksman_Revolver_Fire_Close_03}}
 {id=WavePlayer_18, properties {soundWaveAssetPtr=&/Marksman_Revolver_Fire_Close_04}}
 {id=WavePlayer_19, properties {soundWaveAssetPtr=&/Marksman_Revolver_Fire_Med_01}}
 {id=WavePlayer_20, properties {soundWaveAssetPtr=&/Marksman_Revolver_Fire_Med_02}}
 {id=WavePlayer_21, properties {soundWaveAssetPtr=&/Marksman_Revolver_Fire_Med_03}}
 {id=WavePlayer_22, properties {soundWaveAssetPtr=&/Marksman_Revolver_Fire_Far_01}}
 {id=WavePlayer_23, properties {soundWaveAssetPtr=&/Marksman_Revolver_Fire_Far_02}}
 {id=WavePlayer_24, properties {soundWaveAssetPtr=&/Marksman_Revolver_Fire_Far_03}}
 {id=WavePlayer_25, properties {soundWaveAssetPtr=&Marksman_Revolver_Fire_Close_01}}
 {id=WavePlayer_26, properties {soundWaveAssetPtr=&Marksman_Revolver_Fire_Close_02}}
 {id=WavePlayer_29, properties {soundWaveAssetPtr=&Marksman_Revolver_Fire_Close_03}}
 {id=WavePlayer_32, properties {soundWaveAssetPtr=&Marksman_Revolver_Fire_Med_01}}
 {id=WavePlayer_33, properties {soundWaveAssetPtr=&Marksman_Revolver_Fire_Med_02}}
 {id=WavePlayer_35, properties {soundWaveAssetPtr=&Marksman_Revolver_Fire_Far_01}}
]
properties {
 maxDistance=20000
 volumeMultiplier=0.3
 attenuationOverrides {falloffDistance=4100, attenuationShapeExtents {x=900}}
 concurrencySet [&Weapon_Gunfire_Local_Conc, &Weapon_Gunfire_Global_Conc]
}
