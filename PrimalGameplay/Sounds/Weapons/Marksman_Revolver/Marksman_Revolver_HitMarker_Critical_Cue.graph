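# Critical-hit marker sound cue for the Marksman Revolver, using a 2D hit-notification sound class.
# A quality-level node selects full layering, a trimmed version with fewer random variants, or one quieter optimized wave shared by the last two levels.
# Layered versions mix a random flesh impact and shared critical layer with a slightly delayed, pitched-up random revolver critical sting.
# Concurrency is set to two voices, stopping the oldest.
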SoundCue /PrimalGameplay/Sounds/Weapons/Marksman_Revolver/Marksman_Revolver_HitMarker_Critical_Cue
root=QualityLevel_0
nodes [
 {id=Delay_0, children [Modulator_2], properties {delayMax=0.15, delayMin=0.15}}
 {id=Delay_1, children [Modulator_1], properties {delayMax=0.15, delayMin=0.15}}
 {id=Mixer_0, children [Delay_1,Modulator_6], properties {inputVolume [1,1]}}
 {id=Mixer_1, children [Random_7,WavePlayer_1], properties {inputVolume [1,1.25]}}
 {id=Mixer_2, children [Delay_0,Modulator_0], properties {inputVolume [1,1]}}
 {id=Mixer_3, children [Random_0, WavePlayer_11], properties {inputVolume [1,1.25]}}
 {id=Modulator_0, children [Mixer_3], properties {pitchMax=1, pitchMin=1}}
 {id=Modulator_1, children [Random_2], properties {pitchMax=2, pitchMin=2}}
 {id=Modulator_2, children [Random_1], properties {pitchMax=2, pitchMin=1.85}}
 {id=Modulator_3, children [WavePlayer_0], properties {volumeMax=0.65, volumeMin=0.65}}
 {id=Modulator_6, children [Mixer_1], properties {pitchMax=1, pitchMin=1}}
 {id=QualityLevel_0, children [Mixer_0,Mixer_2,Modulator_3,Modulator_3]}
 {id=Random_0, children [WavePlayer_2,WavePlayer_3], properties {weights [1,1]}}
 {id=Random_1, children [WavePlayer_9, WavePlayer_10], properties {weights [1,1]}}
 {id=Random_2, children [WavePlayer_4,WavePlayer_5,WavePlayer_8], properties {weights [1,1,1]}}
 {id=Random_7, children [WavePlayer_20, WavePlayer_21, WavePlayer_22, WavePlayer_23], properties {weights [1,1,1,1]}}
 {id=WavePlayer_0, properties {soundWaveAssetPtr=/Game/Athena/Sounds/HitNotifications/HitMarker_Critical_opt}}
 {id=WavePlayer_1, properties {soundWaveAssetPtr=/Game/Athena/Sounds/HitNotifications/sfx_hitmarker_notification_critical_layer}}
 {id=WavePlayer_10, properties {soundWaveAssetPtr=&/Marksman_Revolver_Critical_Hit_02}}
 {id=WavePlayer_11, properties {soundWaveAssetPtr=&sfx_hitmarker_notification_critical_layer}}
 {id=WavePlayer_2, properties {soundWaveAssetPtr=/Game/Sounds/Fort_Impact_Sounds/Flesh/player_impact_enemy_stereo_01}}
 {id=WavePlayer_20, properties {soundWaveAssetPtr=&player_impact_enemy_stereo_01}}
 {id=WavePlayer_21, properties {soundWaveAssetPtr=/Game/Sounds/Fort_Impact_Sounds/Flesh/player_impact_enemy_stereo_02}}
 {id=WavePlayer_22, properties {soundWaveAssetPtr=/Game/Sounds/Fort_Impact_Sounds/Flesh/player_impact_enemy_stereo_03}}
 {id=WavePlayer_23, properties {soundWaveAssetPtr=/Game/Sounds/Fort_Impact_Sounds/Flesh/player_impact_enemy_stereo_04}}
 {id=WavePlayer_3, properties {soundWaveAssetPtr=&player_impact_enemy_stereo_02}}
 {id=WavePlayer_4, properties {soundWaveAssetPtr=&/Marksman_Revolver_Critical_Hit_01}}
 {id=WavePlayer_5, properties {soundWaveAssetPtr=&Marksman_Revolver_Critical_Hit_02}}
 {id=WavePlayer_8, properties {soundWaveAssetPtr=&/Marksman_Revolver_Critical_Hit_03}}
 {id=WavePlayer_9, properties {soundWaveAssetPtr=&Marksman_Revolver_Critical_Hit_01}}
]
properties {
 maxDistance=2097152
 soundClassObject=&Fort_2DHitNotifications_Instigate_PSM
 volumeMultiplier=0.4
 concurrencyOverrides {maxCount=2, resolutionRule=StopOldest}
 concurrencySet [&HitNotifications_TwoMaximum_StopOldest_VolumeScale]
}
