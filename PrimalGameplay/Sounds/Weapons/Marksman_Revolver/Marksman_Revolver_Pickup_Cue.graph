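# Sound cue for the Marksman Revolver pickup sound, classed as loot collection and spatialized with the shared weapon pickup attenuation.
# A quality-level root routes the first two tiers to one mixer and the last two to another, each boosting a single pickup wave player.
# Playback is limited through the shared PickupMaximum_Oldest concurrency set, with single-instance, prevent-new override values also stored on the cue.
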SoundCue /PrimalGameplay/Sounds/Weapons/Marksman_Revolver/Marksman_Revolver_Pickup_Cue
root=QualityLevel_1
nodes [
 {id=Mixer_0, children [WavePlayer_1], properties {inputVolume [2]}}
 {id=Mixer_1, children [WavePlayer_3], properties {inputVolume [2]}}
 {id=QualityLevel_1, children [Mixer_0,Mixer_0,Mixer_1,Mixer_1]}
 {id=WavePlayer_1, properties {soundWaveAssetPtr=&/Marksman_Revolver_Pickup}}
 {id=WavePlayer_3, properties {soundWaveAssetPtr=&Marksman_Revolver_Pickup}}
]
properties {
 attenuationSettings=&Fort_Weapon_Pickup_Att
 maxDistance=3500
 soundClassObject=&Fort_Loot_Collect
 volumeMultiplier=0.4615
 concurrencyOverrides {maxCount=1, resolutionRule=PreventNew}
 concurrencySet [&PickupMaximum_Oldest]
}
