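# Sound cue for the end of the Marksman Revolver's reload, played as weapon foley.
# A single wave player feeds the reload-end sound wave directly to the output at reduced volume.
# It uses the shared weapon reload attenuation and foley sound class, with a concurrency override that prevents new instances rather than cutting off playing ones.
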SoundCue /PrimalGameplay/Sounds/Weapons/Marksman_Revolver/Marksman_Revolver_Reload_End_Cue
root=WavePlayer_0
nodes [{id=WavePlayer_0, properties {soundWaveAssetPtr=&/Marksman_Revolver_Reload_End}}]
properties {
 attenuationSettings=&Fort_Weapon_Reload_Att
 bOverrideConcurrency=true
 maxDistance=2250
 soundClassObject=&Fort_WeaponFoley
 volumeMultiplier=0.501
 concurrencyOverrides {resolutionRule=PreventNew}
}
