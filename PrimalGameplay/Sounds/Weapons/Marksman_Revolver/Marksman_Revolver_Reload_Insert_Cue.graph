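# Sound cue for the Marksman Revolver's reload insert foley in weapon audio.
# A single wave player plays the insert sound through a modulator that slightly randomizes pitch to vary repeats.
# It routes to the weapon foley sound class with weapon reload attenuation and reduced volume.
# A concurrency override prevents new instances from starting once the limit is reached.
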SoundCue /PrimalGameplay/Sounds/Weapons/Marksman_Revolver/Marksman_Revolver_Reload_Insert_Cue
root=Modulator_0
nodes [
 {id=Modulator_0, children [WavePlayer_0], properties {pitchMax=1.02, pitchMin=0.98}}
 {id=WavePlayer_0, properties {soundWaveAssetPtr=&/Marksman_Revolver_Reload_Insert}}
]
properties {
 attenuationSettings=&Fort_Weapon_Reload_Att
 bOverrideConcurrency=true
 maxDistance=2250
 soundClassObject=&Fort_WeaponFoley
 volumeMultiplier=0.501
 concurrencyOverrides {resolutionRule=PreventNew}
}
