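# Sound cue for the start of the Marksman Revolver reload, played as weapon foley with reload attenuation.
# A mixer layers a louder, slightly pitch- and volume-modulated random pick of two revolver reload-start waves with a quieter random pick of two SixShooter reload waves.
# The SixShooter layer is delayed by a fixed half second.
# A concurrency override prevents new instances from starting once the limit is reached.
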SoundCue /PrimalGameplay/Sounds/Weapons/Marksman_Revolver/Marksman_Revolver_Reload_Start_Cue
root=Mixer_1
nodes [
 {id=Delay_1, children [Random_0], properties {delayMax=0.5, delayMin=0.5}}
 {id=Mixer_1, children [Modulator_0,Delay_1], properties {inputVolume [1.4,0.4]}}
 {id=Modulator_0, children [Random_1], properties {pitchMax=1.1, pitchMin=1, volumeMax=1.2, volumeMin=1.1}}
 {id=Random_0, children [WavePlayer_3,WavePlayer_4], properties {weights [1,1]}}
 {id=Random_1, children [WavePlayer_0,WavePlayer_1], properties {weights [1,1]}}
 {id=WavePlayer_0, properties {soundWaveAssetPtr=&/Marksman_Revolver_Reload_Start_01}}
 {id=WavePlayer_1, properties {soundWaveAssetPtr=&/Marksman_Revolver_Reload_Start_02}}
 {id=WavePlayer_3, properties {soundWaveAssetPtr=/Game/Sounds/Fort_Weapons/pistol/SixShooter/SixShooter_ReloadStartB_01}}
 {id=WavePlayer_4, properties {soundWaveAssetPtr=/Game/Sounds/Fort_Weapons/pistol/SixShooter/SixShooter_ReloadStartB_02}}
]
properties {
 attenuationSettings=&Fort_Weapon_Reload_Att
 bOverrideConcurrency=true
 maxDistance=2250
 soundClassObject=&Fort_WeaponFoley
 volumeMultiplier=0.376
 concurrencyOverrides {resolutionRule=PreventNew}
}
